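# Niagara ribbon trail effect for the Precision Mongoose pickaxe, driven by the weapon's animation.
# One emitter uses FN_AnimTrail modules to spawn, initialize and update short-lived ribbon particles from the User.SkeletalMesh interface, alongside a ribbon ID offset utility.
# Particles get a bright pale-cyan HDR color plus a curve-scaled Linear Color at spawn, then a curve quickly fades alpha.
# Two ribbon renderers draw the same particles.
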NiagaraSystem /BRCosmetics/Weapons/FORT_Melee/Pickaxe_PrecisionMongoose/FX/NS_Pickaxe_PrecisionMongoose_Trail
dependencies {
 usedDataInterfaces [&NiagaraDataInterfaceColorCurve, &NiagaraDataInterfaceCurve]
 usedDynamicInputs []
 usedModules [&SystemState]
}
emitters {FN_AnimTrail {
  emitterSpawnScript [&FN_RibbonIDOffsetUtility()]
  emitterUpdateScript [
   &EmitterState()
   &FN_SpawnAnimTrail()
  ]
  particleSpawnScript [
   &FN_InitializeAnimTrailBoneIndecies()
   &FN_InitializeAnimTrail(Color=<rgba 1.7031 2.4553 3 0.2>, Lifetime=0.15)
   /SetVariables(Particles.Linear Color=&MultiplyLinearColor(
     Linear Color=<rgba 14.1927 20.2151 25 0.55>
     Scale Factor=&Vector4FromCurve(Vector4Curve=DataInterface() {propertyValues {
        AlphaCurve {keys [<at 0 1>,<at 0.3 0>]}
        BlueCurve {keys [<at 0 1>,<at 0.6935 1>]}
        GreenCurve {keys [<at 0 1>,<at 0.6935 1>]}
        RedCurve {keys [<at 0 1>,<at 0.6935 1>]}
       }})
    )) {isSetParameters=true}
  ]
  particleUpdateScript [
   &SolveForcesAndVelocity(Clamp Velocity=false, Limit Acceleration=false)
   &FN_UpdateAnimTrail(Locked Particle RibbonWidth=1, SkeletalMesh=$User.SkeletalMesh) {module=FN_UpdateAnimTrail001}
   &ScaleColor(
    Scale Alpha=&FloatFromCurve(FloatCurve=DataInterface() {propertyValues {Curve {keys [
         {interpMode=RCIM_Cubic, leaveTangent=-5.4257, tangentMode=RCTM_Break, time=0, value=1}
         {arriveTangent=-0.0992, tangentMode=RCTM_User, time=0.5, value=0}
        ]}}})
    ScaleA=true
    ScaleRGB=true
   )
  ]
  rendererClasses [&NiagaraRibbonRendererProperties]
  renderers [
   {rendererClass=&NiagaraRibbonRendererProperties}
   {rendererClass=&NiagaraRibbonRendererProperties, rendererIndex=1}
  ]
 }}
summary {
 systemName=NS_Pickaxe_PrecisionMongoose_Trail
 emitters [{emitterName=FN_AnimTrail, rendererClasses [&NiagaraRibbonRendererProperties]}]
 userVariables [
  User.RibbonIDOffset {type=NiagaraInt32}
  User.SkeletalMesh {
   type=NiagaraDataInterfaceSkeletalMesh
   defaultValue {
    struct=NiagaraExt_VariableValue_DataInterface
    value {dataInterface=&self:NiagaraDataInterfaceSkeletalMesh_0}
   }
  }
 ]
}
